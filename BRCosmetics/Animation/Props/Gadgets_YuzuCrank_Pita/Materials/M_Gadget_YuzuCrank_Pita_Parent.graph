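# Parent material for the YuzuCrank Pita gadget prop, adding animated wheel dots to the emissive of the standard character material.
# Two offset pre-skinned local-position YZ projections are spun over time, scaled and vertex-interpolated to sample a clamped grayscale mask, one per apparent wheel.
# The summed masks form WheelDots, feeding an emissive boost set on the material attributes between the CharacterParent PreFX and PostFX functions.
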Material /BRCosmetics/Animation/Props/Gadgets_YuzuCrank_Pita/Materials/M_Gadget_YuzuCrank_Pita_Parent
graph {
 properties {
  bUseMaterialAttributes=true
  bUsedWithMeshDeformer=true
  bUsedWithMorphTargets=true
  bUsedWithSkeletalMesh=true
  pixelDepthOffsetMode=PDOM_Legacy
  previewMesh=/BRCosmetics/Animation/Props/Gadgets_YuzuCrank_Pita/Meshes/YuzuCrank_Pita_Gadget
  shadingModel=MSM_FromMaterialExpression
  usedShadingModels=Default Lit | Subsurface | Clear Coat | Subsurface Profile | Cloth
 }
 nodes [
  n1 = Multiply(A=Time(), B=ScalarParameter() {defaultValue=12})
  Frac(None=@n1)
  n2 = MaterialFunctionCall() {materialFunction=&MF_CharacterParent_2023_PreFX}
  MaterialFunctionCall(Material=NamedRerouteUsage(), M=@n2.M, Pre Skinned Local Position=@n2.M, Distance Blend=@n2.M, Fresnel=@n2.M) {materialFunction=&MF_CharacterParent_2023_PostFX}
  NamedRerouteDeclaration(None=@n2.Result) {name=InMaterial, nodeColor=<rgba 0.3125 1 0 1>}
  n3 = GetMaterialAttributes(None=NamedRerouteUsage()) {attributeGetTypes [#1]}
  NamedRerouteDeclaration(None=SetMaterialAttributes(MaterialAttributes=@n3.MaterialAttributes, Emissive Color=NamedRerouteUsage()) {attributeSetTypes [#1]}) {name=OutMaterial, nodeColor=<rgba 1 0 0.6328 1>}
  NamedRerouteDeclaration(None=@n3.Emissive Color) {name=InEmissive, nodeColor=<rgba 0 1 0.5781 1>}
  n4 = Add(A=LocalPosition() {localOrigin=InstancePreSkinning}.XYZ, B=Constant3Vector() {constant=<rgba 0 15 28 0>})
  n5 = Rotator(Coordinate=ComponentMask(None=Multiply(A=@n4) {constB=0.01}) {b=true, g=true}, Time=@n1)
  n6 = Constant() {r=0.33}
  n7 = TextureSample(UVs=VertexInterpolator(VS=MaterialFunctionCall(UVs=@n5, Texture Scale=@n6) {materialFunction=&ScaleUVsByCenter}.UVs).PS) {
   samplerSource=SSM_Clamp_WorldGroupSettings
   samplerType=SAMPLERTYPE_LinearGrayscale
   texture=/BRCosmetics/Animation/Props/Gadgets_YuzuCrank_Pita/Textures/T_YuzuCrank_Pita_Gadget_mask
  }
  n8 = Add(A=@n4, B=Constant3Vector() {constant=<rgba 0 70 0 0>})
  n9 = Rotator(Coordinate=ComponentMask(None=Multiply(A=@n8) {constB=0.01}) {b=true, g=true}, Time=@n1)
  n10 = Constant() {r=0.33}
  n11 = TextureSample(UVs=VertexInterpolator(VS=MaterialFunctionCall(UVs=@n9, Texture Scale=@n10) {materialFunction=&ScaleUVsByCenter}.UVs).PS) {
   samplerSource=SSM_Clamp_WorldGroupSettings
   samplerType=SAMPLERTYPE_LinearGrayscale
   texture=&T_YuzuCrank_Pita_Gadget_mask
  }
  NamedRerouteDeclaration(None=Multiply(A=NamedRerouteUsage(), B=Add(A=@n7.R, B=@n11.R))) {name=WheelDots, nodeColor=<rgba 1 0.2578 0 1>}
  n12 = NamedRerouteUsage()
  n13 = ScalarParameter() {defaultValue=1.5, parameterName=Param_1}
  NamedRerouteDeclaration(None=Add(A=NamedRerouteUsage(), B=Multiply(A=@n12, B=@n13))) {name=OutEmissive, nodeColor=<rgba 0 0.0625 1 1>}
 ]
}
